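# Unlit translucent material for Niagara mesh particles forming the background effect of the TremorMark emote.
# Emissive tints panning smoke and noise-warped plasma-coil lightning with a vertical curve-atlas gradient and particle color, with optional per-particle lightning brightness.
# Opacity combines a vertical sine mask, view-angle fade and panning noise, minus a smoky band, with noisy depth fade and particle-driven HeightLerp dissolve.
# Panning noise scales vertex normals for mesh wobble.
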Material /BRCosmetics/Effects/Fort_Effects/Emotes/TremorMark/Materials/M_Emote_TremorMark_BackGround
graph {
 properties {
  bAutomaticallySetUsageInEditor=false
  bUsedWithNiagaraMeshParticles=true
  blendMode=BLEND_Translucent
  shadingModel=MSM_Unlit
  usedShadingModels=Unlit
 }
 nodes [
  n1 = TextureCoordinate()
  n2 = CurveAtlasRowParameter(CurveTime=OneMinus(None=ComponentMask(None=@n1) {g=true})) {
   atlas=&/CAtlas_Emote_TremorMark
   curve=&/CV_Emote_TremorMark_Color
   parameterName=Color
  }
  n3 = SmoothStep(Min=ComponentMask(None=TextureCoordinate()) {g=true}) {constValue=0.5}
  n4 = ComponentMask(None=Constant4Vector() {constant=<rgba 5 5 0 0>}.RGB) {g=true, r=true}
  n5 = TextureCoordinate() {uTiling=0, vTiling=0}
  n6 = ConstantBiasScale(None=TextureSample(UVs=Panner(Coordinate=@n5) {speedX=0.1}) {texture=&/T_Granular_noise}.R) {bias=-0.6, scale=10}
  n7 = Saturate(None=@n6)
  n8 = LinearInterpolate(Alpha=@n7) {constA=1, constB=1.2}
  n9 = Add(A=Multiply(A=TextureCoordinate(), B=Multiply(A=@n4, B=@n8)), B=Constant())
  n10 = Power(Base=ComponentMask(None=TextureCoordinate()) {g=true}) {constExponent=1.5}
  n11 = Multiply(
   A=TextureSample(UVs=Add(A=@n9, B=@n6)) {samplerType=SAMPLERTYPE_LinearGrayscale, texture=/Game/Effects/Fort_Effects/Emotes/UltraLight/Texture/T_PlasmaCoils}.R
   B=Saturate(None=Multiply(A=@n10) {constB=50})
  )
  n12 = Add(A=TextureCoordinate() {uTiling=2, vTiling=2}, B=@n7)
  n13 = Multiply(
   A=Multiply(A=Constant4Vector() {constant=<rgba 1.5 0.03 0.0339 1>}.RGB, B=Power(Base=@n11) {constExponent=2.3})
   B=Power(Base=TextureSample(UVs=@n12) {samplerType=SAMPLERTYPE_Grayscale, texture=/BRCosmetics/Weapons/FORT_Melee/Pickaxe_Cube_Ninja_Male/FX/T_CellNoise}.R)
  )
  n14 = Constant() {r=130}
  n15 = LinearInterpolate(
   B=DynamicParameter() {defaultValue=<rgba 0 1 1 1>, paramNames [Dissolve, Param2, Param3, LightningBright]}.LightningBright
   Alpha=ScalarParameter() {parameterName=UseCustomLightningBright}
  ) {constA=1}
  n16 = Multiply(A=Multiply(A=Multiply(A=@n13, B=@n7), B=Multiply(A=@n14, B=@n15)) {constB=170}, B=Saturate(None=Multiply(A=@n3) {constB=5}))
  n17 = TextureCoordinate() {uTiling=2, vTiling=3}
  n18 = TextureSample(UVs=Panner(Coordinate=TextureCoordinate() {uTiling=3, vTiling=3})) {texture=/Game/Effects/Niagara/MasterMaterials/SplineThickenTrail/Texture/T_Pickaxe_SoapPocket_NoiseGlow}
  n19 = Constant() {r=0.1}
  n20 = Panner(Coordinate=BumpOffset(Coordinate=Add(A=@n17) {constB=0}, Height=@n18.R, HeightRatioInput=@n19) {heightRatio=0}) {speedX=0.07}
  n21 = Add(A=TextureCoordinate() {uTiling=3, vTiling=4}) {constB=0}
  n22 = TextureSample(UVs=Panner(Coordinate=BumpOffset(Coordinate=@n21, Height=@n18.R, HeightRatioInput=@n19) {heightRatio=0}) {speedX=-0.07}) {texture=/Game/Effects/Fort_Effects/Textures/T_Tile_Pattern_01}
  n23 = Multiply(
   A=LinearInterpolate(Alpha=@n7) {constA=1, constB=2}
   B=LinearInterpolate(A=TextureSample(UVs=@n20) {texture=/Game/Global/Textures/Default/Tileable_Patterns/T_Thick_Smoke_05}.R, B=Multiply(A=@n22.R) {constB=0.7})
  )
  n24 = TextureCoordinate()
  n25 = Multiply(A=SmoothStep(Min=ComponentMask(None=@n24) {g=true}) {constValue=0.6}) {constB=5}
  n26 = Multiply(A=@n2, B=LinearInterpolate(A=Constant() {r=0.3}, B=Add(A=@n16, B=@n23), Alpha=Saturate(None=@n25)))
  n27 = ParticleColor()
  n28 = Fresnel() {baseReflectFraction=0, exponent=3}
  MaterialFunctionCall(Emissive=Subtract(A=Multiply(A=@n26, B=@n27.RGB), B=Multiply(A=@n28)) {constB=0}) {materialFunction=&MF_FX_EmissiveAdjustments}
  n29 = ComponentMask(None=TextureCoordinate()) {g=true}
  n30 = Multiply(A=Power(Base=Sine(None=@n29) {period=2}) {constExponent=2.5}) {constB=3}
  n31 = TextureCoordinate() {vTiling=2.5}
  n32 = TextureCoordinate() {vTiling=2.5}
  n33 = Multiply(
   A=TextureSample(UVs=Panner(Coordinate=@n31) {speedX=0.1, speedY=0.1}) {texture=/Game/Global/Textures/Default/Tileable_Patterns/T_Blowing_Smoke_Pattern2_Tile_DISP}.R
   B=TextureSample(UVs=Panner(Coordinate=@n32) {speedX=-0.1, speedY=-0.1}) {texture=/BRCosmetics/Weapons/FORT_Melee/Pickaxe_MadameMoth/FX/T_Fibre_Noise}.R
  )
  n34 = MaterialFunctionCall("FadeSensitivity (optional)"=Power(Base=LinearInterpolate(Alpha=@n33) {constA=2, constB=20}) {constExponent=1}) {materialFunction=&FadeBasedOnViewAngle}
  n35 = ScalarParameter() {defaultValue=1, parameterName="Contrast (S)"}
  n36 = TextureSample(UVs=Panner(Coordinate=TextureCoordinate() {uTiling=2, vTiling=2.5}) {speedX=-0.1, speedY=-0.1}) {samplerType=SAMPLERTYPE_Grayscale, texture=&LowResBlurredNoise}
  n37 = Multiply(
   A=Multiply(A=@n27.A, B=Saturate(None=@n30))
   B=Multiply(
    A=MaterialFunctionCall(In=@n34.Result, Contrast=@n35) {materialFunction=&CheapContrast}.Result
    B=LinearInterpolate(Alpha=@n36.R) {constA=0.5}
   )
  )
  n38 = Panner(Coordinate=TextureCoordinate() {uTiling=3, vTiling=1.5}) {speedX=0.05, speedY=0.04}
  n39 = Panner(Coordinate=TextureCoordinate() {uTiling=3, vTiling=2}) {speedX=-0.05, speedY=-0.04}
  n40 = ComponentMask(None=TextureCoordinate()) {g=true}
  n41 = Add(A=@n40) {constB=0.25}
  n42 = Saturate(None=Multiply(A=Sine(None=@n41) {period=0.4}, B=Step(X=@n40) {constX=0.5, constY=0.4}))
  n43 = Power(
   Base=LinearInterpolate(
    A=TextureSample(UVs=@n38) {texture=&T_Tile_Pattern_01}.R
    B=TextureSample(UVs=@n39) {texture=&T_Thick_Smoke_05}.R
   )
   Exp=LinearInterpolate(Alpha=@n42) {constA=3}
  )
  n44 = Power(Base=@n42) {constExponent=3}
  n45 = Saturate(None=Multiply(A=Multiply(A=@n43, B=@n44)) {constB=3})
  n46 = TextureSample() {texture=&T_Inky_Smoke_Tile}
  n47 = LinearInterpolate(Alpha=DynamicParameter() {defaultValue=<rgba 0 1 1 1>, paramNames [Dissolve, Param2, Param3, LightningBright]}.Dissolve)
  n48 = TextureSample(UVs=Panner(Coordinate=TextureCoordinate() {vTiling=1.5}) {speedX=0.05, speedY=0.04}) {texture=&T_Tile_Pattern_01}
  n49 = TextureSample(UVs=Panner(Coordinate=TextureCoordinate() {vTiling=2}) {speedX=-0.05, speedY=-0.04}) {texture=&T_Thick_Smoke_05}
  n50 = Saturate(None=LinearInterpolate(Alpha=LinearInterpolate(A=@n48.R, B=@n49.R)) {constA=-0.2})
  n51 = ScalarParameter() {defaultValue=0.1, parameterName=DissolveContrast}
  Multiply(
   A=DepthFade(Opacity=Subtract(A=@n37, B=@n45), FadeDistance=LinearInterpolate(Alpha=@n46.R) {constA=50, constB=200})
   B=Multiply(A=MaterialFunctionCall(Transition Phase=@n47, Height Texture=@n50, Contrast=@n51) {materialFunction=&HeightLerp}.Results)
  )
  n52 = Panner() {speedX=0.2, speedY=0.2}
  Multiply(A=Multiply(
    A=VertexNormalWS()
    B=TextureSample(UVs=@n52) {samplerType=SAMPLERTYPE_Grayscale, texture=&LowResBlurredNoise}.R
   )) {constB=30}
  Power(Base=@n13) {constExponent=1}
 ]
}
