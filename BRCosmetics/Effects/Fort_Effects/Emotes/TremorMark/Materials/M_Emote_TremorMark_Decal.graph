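# Translucent deferred decal material for the TremorMark emote, apparently a ground mark built from ripple and noise textures.
# A ripple mask times an exponential radial gradient is clipped by a world-space sphere mask around CenterMask and faded by decal lifetime opacity.
# Two tiled noise normals are lerped, then angle-blended with a ground noise normal, using negative FlattenNormal values to exaggerate relief.
# Scalar parameters supply base color, specular, metallic and roughness values.
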Material /BRCosmetics/Effects/Fort_Effects/Emotes/TremorMark/Materials/M_Emote_TremorMark_Decal
graph {
 properties {
  bEnableResponsiveAA=true
  bOutputTranslucentVelocity=true
  bUsedWithNiagaraRibbons=true
  bUsedWithNiagaraSprites=true
  bUsedWithParticleSprites=true
  blendMode=BLEND_Translucent
  materialDomain=MD_DeferredDecal
  pixelDepthOffsetMode=PDOM_Legacy
  shadingModel=MSM_Unlit
  twoSided=true
  usedShadingModels=Unlit
 }
 nodes [
  n1 = TextureSampleParameter2D() {parameterName=Mask, samplerType=SAMPLERTYPE_LinearColor, texture=/Game/Effects/Fort_Effects/Effects/Water/Textures/WaterDecalRippleMask}
  n2 = MaterialFunctionCall(
   Radius=ScalarParameter() {defaultValue=0.6, parameterName="Radius (S)"}
   Density=ScalarParameter() {defaultValue=1, parameterName="Density (S)"}
   Invert Density=StaticBool()
  ) {materialFunction=&RadialGradientExponential}
  n3 = Multiply(A=@n1.R, B=Multiply(A=Multiply(A=@n2.RadialGradientExponential)) {constB=3}) {constB=2}
  n4 = SphereMask(
   A=WorldPosition().XYZ
   B=VectorParameter() {parameterName=CenterMask}.RGB
   Radius=ScalarParameter() {defaultValue=220, parameterName=Radius}
   Hardness=ScalarParameter() {parameterName=Hardness}
  ) {hardnessPercent=0}
  Multiply(A=Multiply(A=Multiply(A=@n3) {constB=1.5}, B=Power(Base=@n4) {constExponent=1.5}), B=DecalLifetimeOpacity().Opacity)
  n5 = Multiply(A=TextureCoordinate() {uTiling=2.5, vTiling=2.5}, B=ScalarParameter() {defaultValue=1, parameterName=UVScale})
  n6 = Multiply(A=TextureCoordinate() {uTiling=2.5, vTiling=2.5}, B=ScalarParameter() {defaultValue=1, parameterName=UVScale})
  n7 = LinearInterpolate(
   A=MaterialFunctionCall(
    Normal=TextureSample(UVs=@n5) {samplerType=SAMPLERTYPE_Normal, texture=&/T_DecalNoise01}.RGB
    Flatness=Constant() {r=-1}
   ) {materialFunction=&FlattenNormal}.Result
   B=MaterialFunctionCall(
    Normal=TextureSample(UVs=@n6) {samplerType=SAMPLERTYPE_Normal, texture=&/T_DecalNoise02}.RGB
    Flatness=Constant() {r=-1}
   ) {materialFunction=&FlattenNormal}.Result
  ) {constAlpha=0.7}
  n8 = TextureCoordinate()
  n9 = ScalarParameter() {defaultValue=1, parameterName=UVScale}
  n10 = MaterialFunctionCall(
   Normal=TextureSample(UVs=Multiply(A=@n8, B=@n9)) {samplerType=SAMPLERTYPE_Normal, texture=&/T_GroundDecalNoise}.RGB
   Flatness=Constant() {r=-3}
  ) {materialFunction=&FlattenNormal}
  Reroute(None=MaterialFunctionCall(
    Normal=MaterialFunctionCall(BaseNormal=@n7, AdditionalNormal=@n10.Result) {materialFunction=&BlendAngleCorrectedNormals}.Result
    Flatness=ScalarParameter() {defaultValue=-3, parameterName="Flatness (S)_1"}
   ) {materialFunction=&FlattenNormal}.Result)
  ScalarParameter() {parameterName=BaseColor}
  ScalarParameter() {defaultValue=0.5, parameterName=Specular}
  ScalarParameter() {parameterName=Metallic}
  ScalarParameter() {defaultValue=1, parameterName=Roughness}
 ]
}
