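# Translucent unlit Niagara sprite material for the TremorMark emote that distorts and color-splits the scene behind each particle.
# Offset SceneColor samples separate red, green and blue, then get scaled by panning noise and particle color.
# Radially panned noise within a ring mask builds an index-of-refraction value and distortion normal.
# A soft circular mask, particle alpha and depth fade shape opacity, which shading-path switches zero on forward and mobile.
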Material /BRCosmetics/Effects/Fort_Effects/Emotes/TremorMark/Materials/M_Emote_TremorMark_SceneColor
graph {
 properties {
  bAutomaticallySetUsageInEditor=false
  bUsedWithNiagaraSprites=true
  blendMode=BLEND_Translucent
  refractionMethod=RM_IndexOfRefraction
  shadingModel=MSM_Unlit
  usedShadingModels=Unlit
 }
 nodes [
  n1 = OneMinus(None=SphereMask(A=TextureCoordinate(), B=Constant() {r=0.5}) {attenuationRadius=0.3, hardnessPercent=0})
  n2 = Multiply(A=SphereMask(A=TextureCoordinate(), B=Constant() {r=0.5}) {attenuationRadius=0.3, hardnessPercent=0}, B=@n1)
  n3 = MaterialFunctionCall() {materialFunction=&VectorToRadialValue}
  n4 = TextureSample(UVs=Multiply(A=Panner(Coordinate=@n3.Radial Coordinates) {speedY=-1}, B=Constant2Vector() {g=1, r=4})) {samplerType=SAMPLERTYPE_Grayscale, texture=&LowResBlurredNoise}
  LinearInterpolate(Alpha=Multiply(A=Multiply(A=@n2, B=@n4.R), B=ParticleColor().A)) {constA=1, constB=1.2}
  n5 = MaterialFunctionCall() {materialFunction=&VectorToRadialValue}
  n6 = TextureSample(UVs=Multiply(A=Panner(Coordinate=@n5.Radial Coordinates) {speedY=-1}, B=Constant2Vector() {g=1, r=4})) {samplerType=SAMPLERTYPE_Normal, texture=/Game/Characters/Player/Female/Medium/Bodies/F_MED_Bistro_Astronaut/FX/T_GaussianNoise_Normal}
  MaterialFunctionCall(Normal=@n6.RGB, Flatness=Constant()) {materialFunction=&FlattenNormal}
  n7 = Panner(Coordinate=TextureCoordinate() {vTiling=2}) {speedX=0.1}
  n8 = Multiply(A=Constant2Vector() {g=0.01, r=0.01}) {constB=0.2}
  n9 = SceneColor!(OffsetFraction=@n8) {inputMode=OffsetFraction, constInput=<0.02 0.02>}
  n10 = SceneColor() {inputMode=OffsetFraction}
  n11 = SceneColor(OffsetFraction=Multiply(A=@n8) {constB=-1}) {inputMode=OffsetFraction, constInput=<-0.1 -0.1>}
  n12 = Multiply(
   A=LinearInterpolate(Alpha=TextureSample(UVs=@n7) {samplerType=SAMPLERTYPE_Grayscale, texture=&LowResBlurredNoise}.R) {constA=0.5, constB=3}
   B=MaterialFunctionCall(X=ComponentMask(None=@n9.RGB) {r=true}, Y=ComponentMask(None=@n10.RGB) {g=true}, Z=ComponentMask(None=@n11.RGB) {b=true}) {materialFunction=&MakeFloat3}.Result
  ) {constA=1}
  n13 = Constant() {r=1}
  n14 = ParticleColor()
  Multiply(A=ShadingPathSwitch(Default=@n12, Deferred=@n12, Forward=@n13, Mobile=@n13), B=@n14.RGB)
  Power(Base=@n1) {constExponent=1}
  n15 = SphereMask(A=TextureCoordinate(), B=Constant() {r=0.5}) {attenuationRadius=0.4, hardnessPercent=0}
  Saturate(None=Multiply(A=Multiply(A=@n15)) {constB=5})
  n16 = SphereMask(A=TextureCoordinate(), B=Constant() {r=0.5}) {attenuationRadius=0.5, hardnessPercent=0}
  n17 = DepthFade(Opacity=Multiply(A=@n14.A, B=Multiply(A=@n16))) {fadeDistanceDefault=30}
  n18 = Constant()
  ShadingPathSwitch(Default=@n17, Deferred=@n17, Forward=@n18, Mobile=@n18)
 ]
}
