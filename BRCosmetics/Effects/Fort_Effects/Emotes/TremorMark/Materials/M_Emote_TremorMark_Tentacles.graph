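# Masked, lit material for the tentacle mesh particles in the TremorMark emote effect, built for Niagara.
# Dithered opacity combines particle alpha, a world-space sphere mask around CenterMask and a UV-gradient dissolve, with mask radius and dissolve driven by dynamic parameters.
# A world-aligned desert sand normal is flattened for surface detail.
# A time-driven sine along the second UV set, varied by vertex red, builds a wiggle offset along a local-space WPODirection.
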Material /BRCosmetics/Effects/Fort_Effects/Emotes/TremorMark/Materials/M_Emote_TremorMark_Tentacles
graph {
 properties {bAutomaticallySetUsageInEditor=false, bUsedWithNiagaraMeshParticles=true, blendMode=BLEND_Masked, usedShadingModels=Default Lit}
 nodes [
  n1 = ParticleColor()
  n2 = SphereMask(
   A=WorldPosition().XYZ
   B=VectorParameter() {parameterName=CenterMask}.RGB
   Radius=DynamicParameter() {defaultValue=<rgba 1 1 1 150>, paramNames [Param1, Param2, Dissolve, CenterMaskRadius]}.CenterMaskRadius
   Hardness=ScalarParameter() {defaultValue=0.5, parameterName=Hardness}
  ) {hardnessPercent=0}
  n3 = TextureCoordinate()
  n4 = DynamicParameter() {defaultValue=<rgba 1 1 1 150>, paramNames [Param1, Param2, Dissolve, CenterMaskRadius]}
  n5 = Multiply(A=Add(A=ComponentMask(None=@n3) {g=true}, B=LinearInterpolate(Alpha=@n4.Dissolve) {constA=-1})) {constB=5}
  MaterialFunctionCall(Alpha Threshold=Multiply(A=Multiply(A=@n1.A, B=@n2), B=Saturate(None=@n5))) {materialFunction=&DitherTemporalAA}
  MaterialFunctionCall(
   Normal=MaterialFunctionCall(
    TextureObject=TextureObject() {samplerType=SAMPLERTYPE_Normal, texture=/Game/Environments/Sets/Arid/Textures/DesertSand_normal}
    TextureSize=ScalarParameter() {defaultValue=500}
   ) {materialFunction=&WorldAlignedTexture}.XYZ Texture
   Flatness=ScalarParameter() {parameterName="Flatness (S)"}
  ) {materialFunction=&FlattenNormal}
  n6 = Transform(None=VectorParameter() {parameterName=WPODirection, defaultValue=<rgba 0 2.5 0 0>}.RGB) {transformSourceType=TRANSFORMSOURCE_Local}
  n7 = LinearInterpolate(Alpha=VertexColor().R) {constA=1, constB=2}
  n8 = ScalarParameter() {defaultValue=3, parameterName=WPO_Tile}
  n9 = Add(A=Multiply(A=TextureCoordinate() {coordinateIndex=1}, B=Multiply(A=@n7, B=@n8) {constB=3}), B=VertexColor().R) {constB=0}
  n10 = Time()
  n11 = ScalarParameter() {defaultValue=1.5, parameterName=Speed}
  n12 = Sine(None=Add(A=ComponentMask(None=@n9) {g=true}, B=Multiply(A=@n10, B=@n11))) {period=2.5}
  Reroute(None=Add(A=Multiply(A=@n6, B=@n12)) {constB=0})
 ]
}
