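# Niagara system for the TremorMark emote, a timed atmospheric effect of glows, smoke and drifting spores around the character.
# A background mesh follows owner rotation, alongside scene-color and glow sprites and paired red and cyan lights, with array lookups varying paired particles.
# Region-masked smoke and two GPU spore emitters with curl noise add volume and motion.
# Platform, camera-distance and front-end functions adjust opacity, size and offsets for readability and scalability.
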NiagaraSystem /BRCosmetics/Effects/Fort_Effects/Emotes/TremorMark/NS_Emote_TremorMark
dependencies {
 usedDataInterfaces [
  &NiagaraDataInterfaceCurve
  &NiagaraDataInterfacePlatformSet
  &NiagaraDataInterfaceArrayFloat3
  &NiagaraDataInterfaceArrayColor
  &NiagaraDataInterfaceArrayFloat
  &NiagaraDataInterfaceSpriteRendererInfo
 ]
 usedDynamicInputs [
  &MakeCustomFloatFromBool
  &SetBoolByFloatComparison
  &VectorLength
 ]
 usedModules [&SystemState]
}
emitters {
 BackGround {
  emitterUpdateScript [
   &EmitterState(Life Cycle Mode=::NewEnumerator1, Loop Behavior=::NewEnumerator1, Loop Duration=1, UseLoopDelay=false)
   &SpawnBurst_Instantaneous(Use Spawn Probability=false, UseLoopCountLimit=false)
  ]
  particleSpawnScript [
   &InitializeParticle(
    Color=<rgba 0.6579 0.913 1 1>
    Lifetime=9.5
    Mesh Scale Mode=::NewEnumerator3
    Mesh Uniform Scale=4.2
    Position Offset=<0 0 100>
    Position Offset Coordinate Space=::Local
    UsePositionOffset=true
    Write Material Random=true
   )
   &CameraOffset(Camera Offset Amount=-20)
  ]
  particleUpdateScript [
   &ParticleState()
   &ScaleColor(
    Scale Alpha=&Multiply_Float(
     A=&FN_PickFloatDependOnPlatform(TargetPlatformActive=3, TargetPlatformNoActive=1)
     B=&FN_FrontEndFloatModifier(
      Gameplay=&FN_LOD_Distance_Remap(Start Distance=200, End Distance=300, Minimum Output=0.1, Maximum Output=1, Clamp Output=true)
      Front End=1
     )
    )
    ScaleA=true
    ScaleRGB=true
   )
   &InheritSourceMovement(ApplyForce=false, ApplyPosition=true, ApplyPositionFalloff=false)
   &SolveForcesAndVelocity(Clamp Velocity=false, Limit Acceleration=false)
   &DynamicMaterialParameters(
    Index 0 Param 1=&FloatFromCurve(FloatCurve=DataInterface() {propertyValues {Curve {keys [<at 0.2 1>,<at 0.45 0>,<at 0.8 0.0029>,<at 0.9 1>]}}})
    Index 0 Param 4=&FN_LOD_Distance_Remap(Start Distance=200, End Distance=400, Minimum Output=0.05, Maximum Output=1, Clamp Output=true)
    Module.Index0Float0Write=true
    Module.Index0Float1Write=true
    Module.Index0Float2Write=true
    Module.Index0Float3Write=true
   )
   &InterpolateOverTime(
    Initial Vector Value:Vector3f=<0 0 1>
    Rate Of Change=2
    Target Vector Value:Vector3f=&RotateVectorByQuaternion(Vector To Rotate=<0 0 1>, Quaternion=$Engine.Owner.Rotation)
    Variable Type=::NewEnumerator2
   )
   /SetVariables(Particles.MeshOrientation=&MakeQuaternion(
     Quaternion From=3
     Axis=<1 0 0>
     Angle=0
     From=<0 0 1>
     To=$Particles.InterpolateOverTime.InterpolateOverTime_V3_.Moving Average
     Angle Type=::NewEnumerator0
     Normalize Vectors=false
    )) {isSetParameters=true}
  ]
  rendererClasses [&NiagaraMeshRendererProperties]
  renderers [{rendererClass=&NiagaraMeshRendererProperties, rendererIndex=0}]
 }
 Glow001 {
  emitterUpdateScript [
   &EmitterState(Life Cycle Mode=::NewEnumerator1, Loop Behavior=::NewEnumerator1, Loop Delay=1.5, Loop Duration=2, UseLoopDelay=true)
   &SpawnBurst_Instantaneous(Spawn Count=2, Use Spawn Probability=false, UseLoopCountLimit=false)
  ]
  particleSpawnScript [
   &PartitionParticles()
   &InitializeParticle(
    Color=&SelectLinearColorFromArray(
     Array Sampling Mode=::NewEnumerator1
     Direct Array Index=$Particles.PartitionParticles.Partition
     Color Selection Array=DataInterface() {propertyValues {ColorData [<rgba 0.0061 0.0413 0.05 1>, <rgba 0.1 0 0.0023 1>]}}
     Random Seed=?
    )
    Lifetime=7.5
    Position Offset=&SelectVectorFromArray(
     Direct Array Index=$Particles.PartitionParticles.Partition
     Vector Selection Array=DataInterface() {propertyValues {FloatData [<0 0 50>, <0 0 180>]}}
    )
    Position Offset Coordinate Space=::Local
    Sprite Size Mode=::NewEnumerator3
    Uniform Sprite Size=&SelectFloatFromArray(Direct Array Index=$Particles.PartitionParticles.Partition, Float Selection Array=DataInterface() {propertyValues {FloatData [600,500]}})
    UsePositionOffset=true
    Write Material Random=true
   )
   &DynamicMaterialParameters(
    Index 0 Param 1=&SelectFloatFromArray(Direct Array Index=$Particles.PartitionParticles.Partition, Float Selection Array=DataInterface() {propertyValues {FloatData [1.5,2]}})
    Index 0 Param 2=&FN_FrontEndFloatModifier(Gameplay=30, Front End=30)
    Module.Index0Float0Write=true
    Module.Index0Float1Write=true
    Module.Index0Float2Write=true
    Module.Index0Float3Write=true
   )
   &CameraOffset(Camera Offset Amount=&SelectFloatFromArray(Direct Array Index=$Particles.PartitionParticles.Partition, Float Selection Array=DataInterface() {propertyValues {FloatData [-30,-100]}}))
  ]
  particleUpdateScript [
   &ParticleState()
   &ScaleSpriteSize(
    Scale Sprite Size Mode=::NewEnumerator0
    Uniform Scale Factor=&FN_FrontEndFloatModifier(
     Gameplay=&FN_PickFloatDependOnPlatform(
      TargetPlatform=DataInterface() {propertyValues {Platforms {
         qualityLevelMask=-2
         deviceProfileStates [
          {profileName=Mobile, qualityLevelMask=0, setQualityLevelMask=4294967295}
          {profileName=Switch, qualityLevelMask=0, setQualityLevelMask=4294967295}
         ]
        }}}
      TargetPlatformActive=1
      TargetPlatformNoActive=0.5
     )
     Front End=1
    )
   )
   &ScaleColor(
    Scale Alpha=&Multiply_Float(
     A=&Lerp_Float(
      A=&FloatFromCurve(
       FloatCurve=DataInterface() {propertyValues {Curve {
          defaultValue=1
          keys [
           {arriveTangent=-0.9545, interpMode=RCIM_Cubic, leaveTangent=-0.9545, tangentMode=RCTM_User, time=0, value=0}
           {arriveTangent=0.9766, interpMode=RCIM_Cubic, leaveTangent=0.9766, time=0.2, value=0.7097}
           {arriveTangent=0.1108, interpMode=RCIM_Cubic, leaveTangent=0.1108, tangentMode=RCTM_User, time=0.8, value=0.7813}
           {interpMode=RCIM_Cubic, tangentMode=RCTM_User, time=0.9, value=0}
          ]
         }}}
       Scale Curve=&FN_PickFloatDependOnPlatform(TargetPlatformActive=3, TargetPlatformNoActive=0.5)
      )
      B=&FloatFromCurve(
       FloatCurve=DataInterface() {propertyValues {Curve {
          defaultValue=1
          keys [
           {arriveTangent=-0.9545, interpMode=RCIM_Cubic, leaveTangent=-0.9545, tangentMode=RCTM_User, time=0, value=0}
           {arriveTangent=1.0417, interpMode=RCIM_Cubic, leaveTangent=1.0417, time=0.2, value=0.7}
           {arriveTangent=0.1108, interpMode=RCIM_Cubic, leaveTangent=0.1108, tangentMode=RCTM_User, time=0.75, value=0.7813}
           {interpMode=RCIM_Cubic, tangentMode=RCTM_User, time=0.9, value=0}
          ]
         }}}
       Scale Curve=&FN_PickFloatDependOnPlatform(TargetPlatformActive=2.5, TargetPlatformNoActive=1.5)
      )
      Alpha=&SelectFloatFromArray(Direct Array Index=$Particles.PartitionParticles.Partition, Float Selection Array=DataInterface() {propertyValues {FloatData [0,1]}})
     )
     B=&Multiply_Float(
      A=&FN_PickFloatDependOnPlatform(TargetPlatformActive=1, TargetPlatformNoActive=0.7)
      B=&FN_FrontEndFloatModifier(
       Gameplay=&FN_LOD_Distance_Remap(Start Distance=200, End Distance=300, Minimum Output=0.2, Maximum Output=1, Clamp Output=true)
       Front End=1
      )
     )
    )
    ScaleA=true
    ScaleRGB=true
   )
  ]
  rendererClasses [&NiagaraSpriteRendererProperties]
  renderers [{rendererClass=&NiagaraSpriteRendererProperties}]
 }
 GlowLow {
  emitterUpdateScript [
   &EmitterState(Life Cycle Mode=::NewEnumerator1, Loop Behavior=::NewEnumerator1, Loop Delay=1.5, Loop Duration=2, UseLoopDelay=true)
   &SpawnBurst_Instantaneous(Spawn Count=2, Use Spawn Probability=false, UseLoopCountLimit=false)
  ]
  particleSpawnScript [
   &PartitionParticles()
   &InitializeParticle(
    Color=<rgba 0 0.0533 0.1 1>
    Lifetime=7.5
    Position Offset=<0 0 80>
    Position Offset Coordinate Space=::Local
    Sprite Size Mode=::NewEnumerator3
    Uniform Sprite Size=500
    UsePositionOffset=true
    Write Material Random=true
   )
   &DynamicMaterialParameters(
    Index 0 Param 1=1.5
    Index 0 Param 2=30
    Module.Index0Float0Write=true
    Module.Index0Float1Write=true
    Module.Index0Float2Write=true
    Module.Index0Float3Write=true
   )
   &CameraOffset(Camera Offset Amount=-30)
  ]
  particleUpdateScript [
   &ParticleState()
   &ScaleSpriteSize(
    Scale Sprite Size Mode=::NewEnumerator0
    Uniform Scale Factor=&FN_FrontEndFloatModifier(
     Gameplay=&FN_PickFloatDependOnPlatform(
      TargetPlatform=DataInterface() {propertyValues {Platforms {
         qualityLevelMask=-4
         deviceProfileStates [
          {profileName=Windows, qualityLevelMask=2, setQualityLevelMask=2}
          {profileName=Mac, qualityLevelMask=2, setQualityLevelMask=2}
          {profileName=Linux, qualityLevelMask=2, setQualityLevelMask=2}
         ]
        }}}
      TargetPlatformActive=1
      TargetPlatformNoActive=0.4
     )
     Front End=1
    )
   )
   &ScaleColor(
    Scale Alpha=&FloatFromCurve(
     FloatCurve=DataInterface() {propertyValues {Curve {
        defaultValue=1
        keys [
         {arriveTangent=-0.9545, interpMode=RCIM_Cubic, leaveTangent=-0.9545, tangentMode=RCTM_User, time=0, value=0}
         {arriveTangent=0.9766, interpMode=RCIM_Cubic, leaveTangent=0.9766, time=0.2, value=0.7097}
         {arriveTangent=0.1108, interpMode=RCIM_Cubic, leaveTangent=0.1108, tangentMode=RCTM_User, time=0.8, value=0.7813}
         {interpMode=RCIM_Cubic, tangentMode=RCTM_User, time=1, value=0}
        ]
       }}}
     Scale Curve=&Multiply_Float(
      A=0.3
      B=&FN_PickFloatDependOnPlatform(TargetPlatform=DataInterface() {propertyValues {Platforms {qualityLevelMask=-1}}}, TargetPlatformActive=1, TargetPlatformNoActive=1)
     )
    )
    ScaleA=true
    ScaleRGB=true
   )
  ]
  rendererClasses [&NiagaraSpriteRendererProperties]
  renderers [{rendererClass=&NiagaraSpriteRendererProperties}]
 }
 Light {
  emitterUpdateScript [
   &EmitterState(Life Cycle Mode=::NewEnumerator1, Loop Behavior=::NewEnumerator1, Loop Delay=1.5, UseLoopDelay=true)
   &SpawnBurst_Instantaneous(Spawn Count=2, Use Spawn Probability=false, UseLoopCountLimit=false)
  ]
  particleSpawnScript [
   &PartitionParticles()
   &InitializeParticle(Color Mode=::NewEnumerator0, Lifetime=7.5, Position Offset=<0 -100 50>, UsePositionOffset=false, Write Material Random=true)
  ]
  particleUpdateScript [
   &ParticleState()
   &FN_Light_Attributes(
    Color=&SelectLinearColorFromArray(
     Array Sampling Mode=::NewEnumerator1
     Direct Array Index=$Particles.PartitionParticles.Partition
     Color Selection Array=DataInterface() {propertyValues {ColorData [<rgba 10 1.2114 0.9869 1>, <rgba 0 15.2696 20 1>]}}
     Random Seed=?
    )
    Create Color Attribute=true
    Create Exponent Attribute=true
    Create Position Attribute=true
    Create Radius Attribute=true
    Exponent=3
    Position=&AddVectorToPosition(
     Position=$Particles.Position
     Vector=&SelectVectorFromArray(
      Direct Array Index=$Particles.PartitionParticles.Partition
      Vector Selection Array=DataInterface() {propertyValues {FloatData [<0 -50 100>, <0 50 100>]}}
     )
    )
    Radius=200
    Scale Color=&FloatFromCurve(FloatCurve=DataInterface() {propertyValues {Curve {keys [<at 0 0>,<at 0.1949 1>,<at 0.8052 1>,<at 1 0>]}}})
   )
  ]
  rendererClasses [&NiagaraLightRendererProperties]
  renderers [{rendererClass=&NiagaraLightRendererProperties, rendererIndex=0}]
 }
 SceneColor {
  emitterUpdateScript [
   &EmitterState(Life Cycle Mode=::NewEnumerator1, Loop Behavior=::NewEnumerator1, Loop Delay=1.5, Loop Duration=2, UseLoopDelay=true)
   &SpawnBurst_Instantaneous(Use Spawn Probability=false, UseLoopCountLimit=false)
  ]
  particleSpawnScript [
   &InitializeParticle(
    Color=<rgba 0.2 0.6935 1 1>
    Lifetime=7.5
    Position Offset=<0 0 100>
    Sprite Size Mode=::NewEnumerator3
    Uniform Sprite Size=300
    UsePositionOffset=true
    Write Material Random=true
   )
   &DynamicMaterialParameters(
    Index 0 Param 1=2
    Index 0 Param 2=&FN_FrontEndFloatModifier(Gameplay=30, Front End=30)
    Module.Index0Float0Write=true
    Module.Index0Float1Write=true
    Module.Index0Float2Write=true
    Module.Index0Float3Write=true
   )
  ]
  particleUpdateScript [
   &CameraOffset(Camera Offset Amount=&FN_LOD_Distance_Remap(Start Distance=300, End Distance=500, Minimum Output=100, Maximum Output=200, Clamp Output=true))
   &ParticleState()
   &ScaleSpriteSize(
    Scale Sprite Size Mode=::NewEnumerator0
    Uniform Scale Factor=&Multiply_Float(
     A=&FN_FrontEndFloatModifier(Gameplay=1, Front End=1)
     B=&FN_LOD_Distance_Remap(Start Distance=300, End Distance=500, Minimum Output=1, Maximum Output=1.2, Clamp Output=true)
    )
   )
   &ScaleColor(
    Scale Alpha=&Multiply_Float(
     A=0.8
     B=&FloatFromCurve(
      FloatCurve=DataInterface() {propertyValues {Curve {keys [<at 0 0>,<at 0.2 1>,<at 0.8 1>,<at 1 0>]}}}
      Scale Curve=&FN_FrontEndFloatModifier(Gameplay=1, Front End=1)
     )
    )
    Scale RGB=&Multiply_VectorByFloat(Vector=<1 1 1>, Float=&FN_FrontEndFloatModifier(Gameplay=1, Front End=0.7))
    ScaleA=true
    ScaleRGB=true
   )
  ]
  rendererClasses [&NiagaraSpriteRendererProperties]
  renderers [{rendererClass=&NiagaraSpriteRendererProperties}]
 }
 Smoke001 {
  emitterUpdateScript [
   &EmitterState(Life Cycle Mode=::NewEnumerator1, Loop Behavior=::NewEnumerator1, Loop Delay=1.5, UseLoopDelay=true)
   &SpawnBurst_Instantaneous(Spawn Count=10, Use Spawn Probability=false, UseLoopCountLimit=false)
   &SpawnRate(SpawnRate=10, Use Spawn Probability=false)
  ]
  particleSpawnScript [
   &InitializeParticle(
    Lifetime Max=3
    Lifetime Min=3
    Lifetime Mode=::NewEnumerator1
    Sprite Rotation Mode=::NewEnumerator2
    Sprite Size Mode=::NewEnumerator1
    Uniform Sprite Size Max=150
    Uniform Sprite Size Min=150
    UsePositionOffset=false
    Write Material Random=true
   )
   &ShapeLocation(
    MaskSpawnGroup=false
    Offset=&MakeVector(
     X=0
     Y=0
     Z=&FN_PickFloatDependOnPlatform(
      TargetPlatform=DataInterface() {propertyValues {Platforms {
         qualityLevelMask=-4
         deviceProfileStates [
          {profileName=Windows, qualityLevelMask=2, setQualityLevelMask=2}
          {profileName=Mac, qualityLevelMask=2, setQualityLevelMask=2}
          {profileName=Linux, qualityLevelMask=2, setQualityLevelMask=2}
         ]
        }}}
      TargetPlatformActive=80
      TargetPlatformNoActive=120
     )
    )
    Offset Coordinate Space=::Local
    Offset Mode=::NewEnumerator0
    Override Seed=false
    Sphere Radius=&FN_PickFloatDependOnPlatform(TargetPlatform=DataInterface() {propertyValues {Platforms {qualityLevelMask=-2}}}, TargetPlatformActive=120, TargetPlatformNoActive=90)
    Sphere Surface Distribution=1
    UseOwnerScale=true
   )
  ]
  particleUpdateScript [
   &RegionMask(Offset:Vector3f=<0 0 100>, Region Type:ENiagaraDebugDrawShape=::NewEnumerator2, UseOffset=true, UseRotation=false)
   &RegionMask(
    Inner Float=0
    Offset:Vector3f=<0 0 100>
    Offset Space:ENiagaraCoordinateSpace=::Local
    Outer Float=1
    Region Radius=260
    UseOffset=true
    UseRotation=false
   ) {module=RegionMask001}
   &ParticleState()
   &SpriteRotationRate(Rotation Rate=&RandomRangeFloat(Minimum=-20, Maximum=20))
   &InheritSourceMovement(ApplyForce=false, ApplyPosition=true, ApplyPositionFalloff=false)
   &SolveForcesAndVelocity(Clamp Velocity=false, Limit Acceleration=false)
   &SubUVAnimation(
    "Sprite Renderer "=DataInterface() {propertyValues {SpriteRenderer=&self:Smoke001_0.NiagaraSpriteRendererProperties_0}}
    UseEndFrame=false
    UseStartFrame=false
   )
   &ScaleColor(
    Scale Alpha=&FloatFromCurve(
     FloatCurve=DataInterface() {propertyValues {Curve {
        defaultValue=1
        keys [
         {interpMode=RCIM_Cubic, tangentMode=RCTM_User, time=0, value=0}
         {arriveTangent=2.4981, interpMode=RCIM_Cubic, leaveTangent=2.4981, time=0.2, value=1}
         <at 0.4003 1 RCIM_Cubic>
         {interpMode=RCIM_Cubic, tangentMode=RCTM_User, time=0.7654, value=1}
         {tangentMode=RCTM_User, time=1, value=0}
        ]
       }}}
     Scale Curve=&Multiply_Float(
      A=&Multiply_Float(A=1, B=&Lerp_Float(A=1.5, B=1.5, Alpha=$Particles.RegionMask.Float))
      B=&FN_PickFloatDependOnPlatform(
       TargetPlatform=DataInterface() {propertyValues {Platforms {
          qualityLevelMask=-4
          deviceProfileStates [
           {profileName=Windows, qualityLevelMask=3, setQualityLevelMask=3}
           {profileName=Mac, qualityLevelMask=2, setQualityLevelMask=2}
           {profileName=Linux, qualityLevelMask=2, setQualityLevelMask=2}
          ]
         }}}
       TargetPlatformActive=1.1
       TargetPlatformNoActive=&FN_LOD_Distance_Remap(Start Distance=300, End Distance=400, Minimum Output=0, Maximum Output=0.5, Clamp Output=true)
      )
     )
    )
    Scale RGB=&MakeVectorFromLinearColorRGB(Color=&LerpLinearColors(StartColor=<rgba 0.0088 0.022 0.03 1>, EndColor=<rgba 0.0594 0.0847 0.1 1>, LerpFactor=$Particles.RegionMask.Float))
    ScaleA=true
    ScaleRGB=true
   )
   &DynamicMaterialParameters(
    Index 1 Param 4=210
    Index 1 Vector 1:Vector3f=&AddVectorToPosition(
     Position=$Engine.Emitter.SimulationPosition
     Vector=&TransformVector(Vector=<0 0 80>, Destination Space=::World)
    )
    Module.Index0Float0Write=false
    Module.Index0Float1Write=false
    Module.Index0Float2Write=false
    Module.Index0Float3Write=false
    Module.Index1Float3Write=true
    Module.Index1VectorWrite=true
    Write Parameter Index 1=::NewEnumerator3
   )
  ]
  rendererClasses [&NiagaraSpriteRendererProperties]
  renderers [{rendererClass=&NiagaraSpriteRendererProperties}]
 }
 Spore001 {
  simTarget=GPUComputeSim
  emitterUpdateScript [
   &EmitterState(Life Cycle Mode=::NewEnumerator1, Loop Behavior=::NewEnumerator1, Loop Delay=1, Loop Duration=4.5, UseLoopDelay=true)
   &SpawnBurst_Instantaneous(Spawn Count=20, Use Spawn Probability=false, UseLoopCountLimit=false)
   &SpawnRate(
    Spawn Probability=&FloatFromCurve(FloatCurve=DataInterface() {propertyValues {Curve {keys [<at 0 0 RCIM_Cubic>,<at 0.5 1>]}}}, CurveIndex=$Emitter.NormalizedLoopAge)
    SpawnRate=50
    Use Spawn Probability=true
   )
  ]
  particleSpawnScript [
   &InitializeParticle(
    Color=<rgba 1.4 1.8701 2 1>
    Lifetime Max=3
    Lifetime Min=3
    Lifetime Mode=::NewEnumerator1
    Sprite Rotation Mode=::NewEnumerator2
    Sprite Size Mode=::NewEnumerator1
    Uniform Sprite Size Max=1.5
    Uniform Sprite Size Min=1.5
    UsePositionOffset=false
    Write Material Random=true
   )
   &ShapeLocation(
    MaskSpawnGroup=false
    Offset=<0 30 100>
    Offset Coordinate Space=::Local
    Offset Mode=::NewEnumerator0
    Override Seed=false
    Sphere Radius=90
    Sphere Surface Distribution=1
    UseOwnerScale=true
   )
   &AddVelocity(Velocity=<0 -50 0>)
  ]
  particleUpdateScript [
   &ParticleState()
   &ScaleSpriteSize(
    Scale Sprite Size Mode=::NewEnumerator0
    Uniform Scale Factor=&FN_LOD_Distance_Remap(Start Distance=200, End Distance=500, Minimum Output=1, Maximum Output=1.2, Clamp Output=true)
   )
   &SpriteRotationRate(Rotation Rate=&RandomRangeFloat(Minimum=200, Maximum=-200))
   &InheritSourceMovement(ApplyForce=false, ApplyPosition=true, ApplyPositionFalloff=false)
   &CurlNoiseForce(Bias Noise=false, Noise Frequency=100, Noise Strength=50, Pan Noise=false)
   &Drag(Drag=0.8, Use Linear Drag=true, Use Rotational Drag=true)
   &SolveForcesAndVelocity(Clamp Velocity=false, Limit Acceleration=false)
   &DynamicMaterialParameters(
    Index 0 Param 1=&RandomRangeFloat(Minimum=0, Maximum=1)
    Module.Index0Float0Write=true
    Module.Index0Float1Write=true
    Module.Index0Float2Write=true
    Module.Index0Float3Write=true
   )
   &ScaleColor(
    Scale Alpha=&Multiply_Float(
     A=&FloatFromCurve(
      FloatCurve=DataInterface() {propertyValues {Curve {
         defaultValue=1
         keys [
          {interpMode=RCIM_Cubic, tangentMode=RCTM_User, time=0, value=0}
          {arriveTangent=1.4286, interpMode=RCIM_Cubic, leaveTangent=1.4286, time=0.3, value=1}
          {interpMode=RCIM_Cubic, tangentMode=RCTM_User, time=0.7, value=1}
          {tangentMode=RCTM_User, time=1, value=0}
         ]
        }}}
      Scale Curve=&FN_LOD_Distance_Remap(Start Distance=200, End Distance=400, Minimum Output=1, Maximum Output=2, Clamp Output=true)
     )
     B=&FloatFromCurve(FloatCurve=DataInterface() {propertyValues {Curve {keys [<at 1 0>,<at 1.5 0>,<at 2 1>]}}}, CurveIndex=$Emitter.Age)
    )
    ScaleA=true
    ScaleRGB=true
   )
  ]
  rendererClasses [&NiagaraSpriteRendererProperties]
  renderers [{rendererClass=&NiagaraSpriteRendererProperties}]
 }
 SporeStart001 {
  simTarget=GPUComputeSim
  emitterUpdateScript [
   &EmitterState(Life Cycle Mode=::NewEnumerator1, Loop Behavior=::NewEnumerator1, Loop Duration=5.5, UseLoopDelay=true)
   &SpawnBurst_Instantaneous(Spawn Count=3, Spawn Time=0.3, Use Spawn Probability=false, UseLoopCountLimit=false)
   &SpawnRate(
    Spawn Probability=&FloatFromCurve(
     FloatCurve=DataInterface() {propertyValues {Curve {keys [<at 0 0>, {arriveTangent=0.3236, interpMode=RCIM_Cubic, leaveTangent=0.3236, tangentMode=RCTM_User, time=0.1513, value=0.0382}, <at 0.3 1>]}}}
     CurveIndex=$Emitter.NormalizedLoopAge
    )
    SpawnRate=10
    Use Spawn Probability=true
   )
  ]
  particleSpawnScript [
   &InitializeParticle(
    Color=<rgba 1.4 1.8701 2 1>
    Lifetime Max=3
    Lifetime Min=2
    Lifetime Mode=::NewEnumerator1
    Sprite Rotation Mode=::NewEnumerator2
    Sprite Size Mode=::NewEnumerator1
    Uniform Sprite Size Max=5
    Uniform Sprite Size Min=3
    UsePositionOffset=false
    Write Material Random=true
   )
   &ShapeLocation(
    Box Size=&MakeVector(
     X=150
     Y=150
     Z=&FloatFromCurve(
      FloatCurve=DataInterface() {propertyValues {Curve {keys [<at 0 0>, {arriveTangent=100, interpMode=RCIM_Cubic, leaveTangent=100, time=0.5634, value=18.935}, <at 1 100>]}}}
      CurveIndex=$Emitter.Age
     )
    )
    MaskSpawnGroup=false
    Offset=<0 100 120>
    Offset Coordinate Space=::Local
    Offset Mode=::NewEnumerator0
    Override Seed=false
    Shape Primitive=::NewEnumerator2
    Surface Only=false
    UseOwnerScale=true
   )
   &AddVelocity(Velocity=<0 -100 0>)
  ]
  particleUpdateScript [
   &ParticleState()
   &ScaleSpriteSize(Uniform Curve Sprite Scale=DataInterface() {propertyValues {Curve {keys [{tangentMode=RCTM_User, time=0, value=0}, <at 0.1978 1>, <at 0.7079 0.9794>, {tangentMode=RCTM_User, time=1, value=0}]}}})
   &SpriteRotationRate(Rotation Rate=&RandomRangeFloat(Minimum=200, Maximum=-200))
   &InheritSourceMovement(ApplyForce=false, ApplyPosition=true, ApplyPositionFalloff=false)
   &CurlNoiseForce(Bias Noise=false, Noise Frequency=1000, Noise Strength=200, Pan Noise=false)
   &Drag(Drag=0.5, Use Linear Drag=true, Use Rotational Drag=true)
   &SolveForcesAndVelocity(Clamp Velocity=false, Limit Acceleration=false)
   &DynamicMaterialParameters(
    Index 0 Param 1=&RandomRangeFloat(Minimum=0, Maximum=1)
    Module.Index0Float0Write=true
    Module.Index0Float1Write=true
    Module.Index0Float2Write=true
    Module.Index0Float3Write=true
   )
  ]
  rendererClasses [&NiagaraSpriteRendererProperties]
  renderers [{rendererClass=&NiagaraSpriteRendererProperties}]
 }
}
summary {
 systemName=NS_Emote_TremorMark
 emitters [
  {emitterName=SporeStart001, simTarget=GPUComputeSim, rendererClasses [&NiagaraSpriteRendererProperties]}
  {emitterName=BackGround, rendererClasses [&NiagaraMeshRendererProperties]}
  {emitterName=Light, rendererClasses [&NiagaraLightRendererProperties]}
  {emitterName=Glow001, rendererClasses [&NiagaraSpriteRendererProperties]}
  {emitterName=SceneColor, rendererClasses [&NiagaraSpriteRendererProperties]}
  {emitterName=Spore001, simTarget=GPUComputeSim, rendererClasses [&NiagaraSpriteRendererProperties]}
  {emitterName=Smoke001, rendererClasses [&NiagaraSpriteRendererProperties]}
  {emitterName=GlowLow, rendererClasses [&NiagaraSpriteRendererProperties]}
 ]
 userVariables [
  User.bIsFrontEnd {type=NiagaraBool}
  User.bIsFrontEndPreview {type=NiagaraBool}
 ]
}
