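# Niagara effect for the TremorMark emote, creating a dark ground decal and tentacle meshes around the character.
# Two mesh emitters (DecalIdle001, TentaclesMove) spawn on the skeletal mesh, with offsets blended by a system InterpolateOverTime value.
# Owner velocity shifts the decal's spawn and how decal and moving tentacles follow the source.
# Emitter-age curves fade everything out, decal fade varies by platform, and TentaclesMove has separate front-end offsets and velocity.
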NiagaraSystem /BRCosmetics/Effects/Fort_Effects/Emotes/TremorMark/NS_Emote_TremorMark_Tentacle
dependencies {
 usedDataInterfaces [
  &NiagaraDataInterfaceCurve
  &NiagaraDataInterfacePlatformSet
  &NiagaraDataInterfaceArrayFloat3
  &NiagaraDataInterfaceSkeletalMesh
 ]
 usedDynamicInputs [&MakeCustomFloatFromBool]
 usedModules [
  &self:SystemSpawnScript.SystemScriptSource.SystemScriptGraph.NiagaraNodeAssignment_0.SetVariables_#1
  &SystemState
  &InterpolateOverTime
 ]
}
emitters {
 Decal {
  emitterUpdateScript [
   &EmitterState(Life Cycle Mode=::NewEnumerator1, Loop Behavior=::NewEnumerator1, Loop Delay=1.5, Loop Duration=5.5, UseLoopDelay=true)
   &SpawnRate(Spawn Probability=$System.InterpolateOverTime.InterpolateOverTime_Float.Moving Average, SpawnRate=2, Use Spawn Probability=false)
   /SetVariables(Emitter.Position2:NiagaraPosition=$Engine.Emitter.SimulationPosition) {isSetParameters=true, module=SetVariables_#2}
  ]
  particleSpawnScript [
   &InitializeParticle(
    Lifetime=2
    Position Offset=&MakeCustomVectorFromBool(
     Bool=&SetBoolByFloatComparison(A=&VectorLength(Vector=$Engine.Owner.Velocity), B=100)
     True Vector=<0 200 0>
     False Vector=<0 0 0>
    )
    Position Offset Coordinate Space=::Local
    UsePositionOffset=true
    Write Material Random=true
   )
   &InitialMeshOrientation(Mesh Axis To Orient=<0 0 1>, Mesh Orientation Mode=::NewEnumerator2, Orientation Vector=<0 0 1>, Use Rotation Vector=false)
   &CameraOffset(Camera Offset Amount=2)
  ]
  particleUpdateScript [
   &ParticleState()
   /Niagara/Modules/Decal/Decal_Attributes(
    Color=1
    Decal Facing:Vector3f=<0 0 0>
    Decal Facing Mode=1
    Decal Fade:NiagaraFloat=&FloatFromCurve(
     FloatCurve=DataInterface() {propertyValues {Curve {keys [<at 0 0>,<at 0.2 1>,<at 0.8 1>,<at 1 0>]}}}
     Scale Curve=&FN_PickFloatDependOnPlatform(
      TargetPlatform=DataInterface() {propertyValues {Platforms {qualityLevelMask=-2}}}
      TargetPlatformActive=&FloatFromCurve(FloatCurve=DataInterface() {propertyValues {Curve {keys [<at 0 1>,<at 7 1>,<at 8 0>]}}}, CurveIndex=$Emitter.Age, Scale Curve=0.8)
      TargetPlatformNoActive=&FloatFromCurve(FloatCurve=DataInterface() {propertyValues {Curve {keys [<at 0 1>,<at 7.5 1>,<at 8.5 0>]}}}, CurveIndex=$Emitter.Age, Scale Curve=0.5)
     )
    )
    Decal Rotation Mode=1
    Decal Size:Vector3f=&Multiply_VectorByFloat(Vector=<5 5 5>, Float=&RandomRangeFloat(Minimum=150, Maximum=150))
    Fade=1
    Orientation=1
    Scale RGB:Vector3f=&Multiply_VectorByFloat(Vector=<1 1 1>, Float=0)
    ScaleA=true
    ScaleRGB=true
    Set Alignment=false
    Size=1
   )
   &InheritSourceMovement(
    Applied Position Scale=&MakeCustomVectorFromBool(
     Bool=&SetBoolByFloatComparison(A=&VectorLength(Vector=$Engine.Owner.Velocity), B=1000)
     True Vector=<0.5 0.5 0.5>
     False Vector=<0 0 0>
    )
    ApplyForce=false
    ApplyPosition=true
    ApplyPositionFalloff=false
   )
   &SolveForcesAndVelocity(Clamp Velocity=false, Limit Acceleration=false)
  ]
  rendererClasses [&NiagaraDecalRendererProperties]
  renderers [{rendererClass=&NiagaraDecalRendererProperties, rendererIndex=0}]
 }
 DecalIdle001 {
  emitterUpdateScript [
   &EmitterState(Life Cycle Mode=::NewEnumerator1, Loop Behavior=::NewEnumerator1, Loop Delay=2, Loop Duration=6, UseLoopDelay=true)
   &SpawnRate(
    SpawnRate=&Lerp_Float(A=15, B=10, Alpha=$System.InterpolateOverTime.InterpolateOverTime_Float.Moving Average)
    Use Spawn Probability=true
   )
   /SetVariables(Emitter.Position2:NiagaraPosition=$Engine.Emitter.SimulationPosition) {isSetParameters=true, module=SetVariables_#2}
   &SpawnBurst_Instantaneous(Spawn Count=10, Spawn Time=1.1, Use Spawn Probability=false, UseLoopCountLimit=false) {module=SpawnBurst_Instantaneous001}
  ]
  particleSpawnScript [
   &PartitionParticles()
   &InitializeParticle(
    Color=<rgba 0.03 0.0066 0.0083 1>
    Lifetime=&Multiply_Float(
     A=&RandomRangeFloat(Minimum=0.9, Maximum=1.5)
     B=&Lerp_Float(A=1, B=1, Alpha=$System.InterpolateOverTime.InterpolateOverTime_Float.Moving Average)
    )
    Mesh Scale Max=<2.5 2.5 2>
    Mesh Scale Min=<1.5 1.5 1>
    Mesh Scale Mode=::NewEnumerator2
    Position Offset=<0 0 2>
    UsePositionOffset=false
    Write Material Random=true
   )
   &InitialMeshOrientation(
    Mesh Axis To Orient=<0 0 1>
    Mesh Orientation Mode=::NewEnumerator2
    Orientation Vector=<0 0 1>
    Rotation=&RandomRangeVector(Minimum=<0 0 0>, Maximum=<0 0 1>)
    Use Rotation Vector=true
   )
   &SkeletalMeshLocation(
    Bone Sampling Mode=::NewEnumerator4
    MaskSpawnGroup=false
    Sampled Position Offset=&Lerp_Vector(
     A=&SelectVectorFromArray(
      Direct Array Index=$Particles.PartitionParticles.Partition
      Vector Selection Array=DataInterface() {propertyValues {FloatData [<0 50 0>, <0 -50 0>]}}
     )
     B=<0 150 0>
     Alpha=$System.InterpolateOverTime.InterpolateOverTime_Float.Moving Average
    )
    Sampled Position Offset Coordinate Space:ENiagaraSkelMeshTransforms=::NewEnumerator2
    Skeletal Mesh=$System.Skeletal Mesh Reader
   )
   /SetVariables(Particles.Meshint32:NiagaraInt32=&RandomRangeInt(Minimum=0, Maximum=2)) {isSetParameters=true}
   &CameraOffset(Camera Offset Amount=&Lerp_Float(A=-10, B=15, Alpha=$System.InterpolateOverTime.InterpolateOverTime_Float.Moving Average))
  ]
  particleUpdateScript [
   &ParticleState()
   &DynamicMaterialParameters(
    Index 0 Param 3=&FloatFromCurve(FloatCurve=DataInterface() {propertyValues {Curve {keys [<at 0 0>,<at 1 1>]}}})
    Index 0 Param 4=150
    Module.Index0Float0Write=true
    Module.Index0Float1Write=true
    Module.Index0Float2Write=true
    Module.Index0Float3Write=true
   )
   &ScaleMeshSize(Scale Factor=&Multiply_VectorByFloat(
     Vector=<1 1 0.7>
     Float=&Multiply_Float(
      A=&RandomRangeFloat(Minimum=0.8, Maximum=1)
      B=&Lerp_Float(A=0.9, B=1, Alpha=$System.InterpolateOverTime.InterpolateOverTime_Float.Moving Average)
     )
    ))
   &InheritSourceMovement(
    Applied Position Scale=&Multiply_VectorByFloat(
     Vector=<1 1 1>
     Float=&FloatFromCurve(FloatCurve=DataInterface() {propertyValues {Curve {keys [<at 0 0.2>,<at 1 0>]}}})
    )
    ApplyForce=false
    ApplyPosition=true
    ApplyPositionFalloff=false
   )
   &SolveForcesAndVelocity(Clamp Velocity=false, Limit Acceleration=false)
   &ScaleColor(
    Scale Alpha=&Multiply_Float(
     A=&FloatFromCurve(FloatCurve=DataInterface() {propertyValues {Curve {keys [<at 0 0>,<at 0.2 1>,<at 0.6 1>,<at 1 0>]}}})
     B=&FloatFromCurve(FloatCurve=DataInterface() {propertyValues {Curve {keys [<at 0 1>,<at 7 1>,<at 8 0>]}}}, CurveIndex=$Emitter.Age)
    )
    ScaleA=true
    ScaleRGB=true
   )
  ]
  rendererClasses [&NiagaraMeshRendererProperties]
  renderers [{rendererClass=&NiagaraMeshRendererProperties, rendererIndex=0}]
 }
 TentaclesMove {
  emitterUpdateScript [
   &EmitterState(Life Cycle Mode=::NewEnumerator1, Loop Behavior=::NewEnumerator1, Loop Delay=2, Loop Duration=6, UseLoopDelay=true)
   &SpawnRate(SpawnRate=3, Use Spawn Probability=true)
   /SetVariables(Emitter.Position2:NiagaraPosition=$Engine.Emitter.SimulationPosition) {isSetParameters=true, module=SetVariables_#2}
  ]
  particleSpawnScript [
   &InitializeParticle(
    Color=<rgba 0.03 0.0066 0.0083 1>
    Mesh Scale Min=<2 2 2>
    Mesh Scale Mode=::NewEnumerator2
    Position Offset=<0 200 3>
    UsePositionOffset=false
    Write Material Random=true
   )
   &InitialMeshOrientation(
    Mesh Axis To Orient=<0 0 1>
    Mesh Orientation Mode=::NewEnumerator2
    Orientation Vector=<0 0 1>
    Rotation=&FN_FrontEndVector3Modifier(
     Gameplay=&Lerp_Vector(
      A=<0 0 0>
      B=&RandomRangeVector(Minimum=<0 0 0>, Maximum=<0 0 1>)
      Alpha=$System.InterpolateOverTime.InterpolateOverTime_Float.Moving Average
     )
     Front End=&RandomRangeVector(Minimum=<0 0 0>, Maximum=<0 0 1>)
    )
    Use Rotation Vector=true
   )
   &DynamicMaterialParameters(
    Index 0 Param 3=1
    Index 0 Param 4=170
    Module.Index0Float0Write=true
    Module.Index0Float1Write=true
    Module.Index0Float2Write=true
    Module.Index0Float3Write=true
   )
   &SkeletalMeshLocation(
    Bone Sampling Mode=::NewEnumerator4
    MaskSpawnGroup=false
    Sampled Position Offset=&FN_FrontEndVector3Modifier(
     Gameplay=&Lerp_Vector(A=<0 0 0>, B=<0 200 2>, Alpha=$System.InterpolateOverTime.InterpolateOverTime_Float.Moving Average)
     Front End=<0 50 0>
    )
    Sampled Position Offset Coordinate Space:ENiagaraSkelMeshTransforms=::NewEnumerator2
    Skeletal Mesh=$System.Skeletal Mesh Reader
   )
   &CameraOffset(Camera Offset Amount=2)
   &AddVelocity(Velocity=&FN_FrontEndVector3Modifier(Gameplay=<0 0 0>, Front End=<0 -100 0>))
  ]
  particleUpdateScript [
   &ParticleState()
   &ScaleMeshSize(Scale Factor=&Multiply_VectorByFloat(Vector=<1 1 1>, Float=2))
   &InheritSourceMovement(
    Applied Position Scale=&MakeCustomVectorFromBool(
     Bool=&SetBoolByFloatComparison(A=&VectorLength(Vector=$Engine.Owner.Velocity), B=1000)
     True Vector=<0.5 0.5 0.5>
     False Vector=<0 0 0>
    )
    ApplyForce=false
    ApplyPosition=true
    ApplyPositionFalloff=false
   )
   &SolveForcesAndVelocity(Clamp Velocity=false, Limit Acceleration=false)
   /SetVariables() {enabled=false, isSetParameters=true}
   &ScaleColor(
    Scale Alpha=&Multiply_Float(
     A=&FloatFromCurve(FloatCurve=DataInterface() {propertyValues {Curve {keys [<at 0 0>,<at 0.2 1>,<at 0.8 1>,<at 1 0>]}}})
     B=&FloatFromCurve(FloatCurve=DataInterface() {propertyValues {Curve {keys [<at 2 0>,<at 3 1.0147>,<at 7 1>,<at 8 0>]}}}, CurveIndex=$Emitter.Age)
    )
    ScaleA=true
    ScaleRGB=true
   )
  ]
  rendererClasses [&NiagaraMeshRendererProperties]
  renderers [{rendererClass=&NiagaraMeshRendererProperties, rendererIndex=0}]
 }
}
summary {
 systemName=NS_Emote_TremorMark_Tentacle
 emitters [
  {emitterName=Decal, rendererClasses [&NiagaraDecalRendererProperties]}
  {emitterName=TentaclesMove, rendererClasses [&NiagaraMeshRendererProperties]}
  {emitterName=DecalIdle001, rendererClasses [&NiagaraMeshRendererProperties]}
 ]
 userVariables [
  User.bIsFrontEnd {type=NiagaraBool}
  User.bIsFrontEndPreview {type=NiagaraBool}
 ]
}
